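# Editor utility widget that gives weapons, pickaxes or specific items during Play In Editor sessions through cheat console commands.
# Three searchable combo boxes hold hard-coded lists of item names or, for specific items, asset paths.
# Each non-empty selection immediately builds a GiveWeapon, GivePickAxe or GiveSpecificItem cheat string.
# The command runs on the first PIE world, excluding dedicated servers, and nothing happens when none exists.
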EditorUtilityWidgetBlueprint /BRCosmetics/Characters/Player/Male/Large/Bodies/M_LRG_ChimneyStack/IK_Retarget_New/EUW_GiveWeapon
graphs [EventGraph {nodes [
   n1 = "On Selection Changed (SearchableStringComboBox_1)"!() -> (OutputDelegate, then, SelectedItem:String, SelectionType:ESelectInfo Enum) {is=title K2Node_ComponentBoundEvent}
   > n2 = Branch(Condition=IsEmpty(InString=@n1.SelectedItem))
   n3 = GetPIEWorlds!(execute=@n2.else, bIncludeDedicatedServer:bool=false) -> (then, ReturnValue:Array of World Object References)
   > Branch(Condition=Is Not Empty!(TargetArray:Array of World Object References=@n3) -> (ReturnValue:bool))
   > ExecuteConsoleCommand(
    WorldContextObject="Get (a copy)"(Array:Array of World Object References=@n3) -> (Output:World Object Reference)
    Command=Append!(A="Cheat GiveSpecificItem ", B=@n1.SelectedItem, C:String=" 1 1") -> (ReturnValue)
   )
   n4 = "On Selection Changed (SearchableStringComboBox_29)"!() -> (OutputDelegate, then, SelectedItem:String, SelectionType:ESelectInfo Enum) {is=title K2Node_ComponentBoundEvent}
   > n5 = Branch(Condition=IsEmpty(InString=@n4.SelectedItem))
   n6 = GetPIEWorlds(execute=@n5.else, bIncludeDedicatedServer=false)
   > Branch(Condition=Is Not Empty(TargetArray=@n6))
   > ExecuteConsoleCommand(WorldContextObject="Get (a copy)"(Array=@n6), Command=Append(A="Cheat GiveWeapon ", B=@n4.SelectedItem, C=" 1 1"))
   n7 = "On Selection Changed (SearchableStringComboBox)"!() -> (OutputDelegate, then, SelectedItem:String, SelectionType:ESelectInfo Enum) {is=title K2Node_ComponentBoundEvent}
   > n8 = Branch(Condition=IsEmpty(InString=@n7.SelectedItem))
   n9 = GetPIEWorlds(execute=@n8.else, bIncludeDedicatedServer=false)
   > Branch(Condition=Is Not Empty(TargetArray=@n9))
   > ExecuteConsoleCommand(WorldContextObject="Get (a copy)"(Array=@n9), Command=Append!(A="Cheat GivePickAxe ", B=@n7.SelectedItem) -> (ReturnValue))
   Event PreConstruct()
   Event Construct()
   Event Tick()
  ]}]
variables []
widgets {
 info {rootWidgetClass=&VerticalBox, widgetCount=7}
 widgets [
  VerticalBox_0 {widgetClassPath=&VerticalBox, widget=VerticalBox_0, properties {bOverride_Cursor=true}}
  TextBlock_23 {
   widgetClassPath=&TextBlock
   parent=VerticalBox_0
   widget=TextBlock_23
   properties {text=Give Weapon}
   slot {class=VerticalBoxSlot, properties {horizontalAlignment=HAlign_Center}}
  }
  SearchableStringComboBox_29 {
   widgetClassPath=/Script/FortniteEditor.SearchableStringComboBox
   bIsVariable=true
   parent=VerticalBox_0
   widget=SearchableStringComboBox_29
   properties {
    defaultOptions [
     "WID_LinedNotebook_Ocean "
     WID_Athena_Bow_Scrap_C
     "WID_Paprika_ClusterBomb "
     "WID_ThermalLance "
     "WID_Assault_CherrySmoke_SpireGun_Athena_C "
     "WID_Athena_TowHookWeapon "
     WID_DualPistol_SemiAuto_Athena_SR_Ore_T03
     WID_CherrySmoke_DMR_Athena_C
     "WID_Linednotebook_Lake "
     WID_Shotgun_FirePetal_Flavor_Athena_C
     WID_SMG_Paprika_DPS_Athena_C
     "WID_FlexLegend_DualSMGs_Athena_C "
     WID_Sniper_Paprika_Athena_R
     WID_Athena_FiendHunter_Dualies_UC
     WID_ChainWhip
     WID_Athena_PurpleMouse
     WID_Launcher_Grenade_Athena_R_Ore_T03
     WID_PowerUp_Scythe_R
     WID_Sawblade_Athena
     WID_JunkGun_R
     WID_FirePetal_Masamune
     WID_Launcher_HomingRocket_Athena_VR
     WID_Launcher_Egg_Athena_VR_Ore
     WID_Assault_LMG_Powerup_Athena_SR
     WID_Assault_LMG_Athena_SR_Ore_T03
     WID_CrimsonMist_HeavyMinigun_SR
     WID_Assault_LMGSAW_Athena_SR_Ore_T03
     "WID_Crossbow_MoonFlax_Athena_SR "
     "WID_UncleBrolly "
    ]
    foregroundColor {colorUseRule=UseColor_Specified, specifiedColor=<rgba 0.5271 0.5271 0.5271 1>}
    itemStyle {
     bUseParentRowBrush=false
     activeBrush {
      bIsDynamicallyLoaded=false
      drawAs=Image
      imageType=NoImage
      mirroring=NoMirror
      resourceName=None
      resourceObject=null
      tiling=NoTile
      imageSize=<0 0>
      margin=<ltrb 0 0 0 0>
      outlineSettings {
       bUseBrushTransparency=false
       roundingType=HalfHeightRadius
       width=0
       color {colorUseRule=UseColor_Specified, specifiedColor=<rgba 0 0 0 0>}
       cornerRadii=<0 0 0 0>
      }
      tintColor {colorUseRule=UseColor_Specified, specifiedColor=<rgba 0 0.162 0.7454 1>}
      uVRegion {bIsValid=false, max=<0 0>, min=<0 0>}
     }
     activeHighlightedBrush {
      bIsDynamicallyLoaded=false
      drawAs=Image
      imageType=NoImage
      mirroring=NoMirror
      resourceName=None
      resourceObject=null
      tiling=NoTile
      imageSize=<0 0>
      margin=<ltrb 0 0 0 0>
      outlineSettings {
       bUseBrushTransparency=false
       roundingType=HalfHeightRadius
       width=0
       color {colorUseRule=UseColor_Specified, specifiedColor=<rgba 0 0 0 0>}
       cornerRadii=<0 0 0 0>
      }
      tintColor {colorUseRule=UseColor_Specified, specifiedColor=<rgba 0.0252 0.0319 0.0423 1>}
      uVRegion {bIsValid=false, max=<0 0>, min=<0 0>}
     }
     activeHoveredBrush {
      bIsDynamicallyLoaded=false
      drawAs=Image
      imageType=NoImage
      mirroring=NoMirror
      resourceName=None
      resourceObject=null
      tiling=NoTile
      imageSize=<0 0>
      margin=<ltrb 0 0 0 0>
      outlineSettings {
       bUseBrushTransparency=false
       roundingType=HalfHeightRadius
       width=0
       color {colorUseRule=UseColor_Specified, specifiedColor=<rgba 0 0 0 0>}
       cornerRadii=<0 0 0 0>
      }
      tintColor {colorUseRule=UseColor_Specified, specifiedColor=<rgba 0 0.162 0.7454 1>}
      uVRegion {bIsValid=false, max=<0 0>, min=<0 0>}
     }
     dropIndicator_Above {
      bIsDynamicallyLoaded=false
      drawAs=Box
      imageType=FullColor
      mirroring=NoMirror
      resourceName=../../../Engine/Content/Slate/Common/DropZoneIndicator_Above.png
      resourceObject=null
      tiling=NoTile
      imageSize=<0 0>
      margin=<ltrb 0.625 0.625 0 0>
      outlineSettings {
       bUseBrushTransparency=false
       roundingType=HalfHeightRadius
       width=0
       color {colorUseRule=UseColor_Specified, specifiedColor=<rgba 0 0 0 0>}
       cornerRadii=<0 0 0 0>
      }
      tintColor {colorUseRule=UseColor_Specified, specifiedColor=<rgba 0.0091 0.3515 0.9647 1>}
      uVRegion {bIsValid=false, max=<0 0>, min=<0 0>}
     }
     dropIndicator_Below {
      bIsDynamicallyLoaded=false
      drawAs=Box
      imageType=FullColor
      mirroring=NoMirror
      resourceName=../../../Engine/Content/Slate/Common/DropZoneIndicator_Below.png
      resourceObject=null
      tiling=NoTile
      imageSize=<0 0>
      margin=<ltrb 0.625 0 0 0.625>
      outlineSettings {
       bUseBrushTransparency=false
       roundingType=HalfHeightRadius
       width=0
       color {colorUseRule=UseColor_Specified, specifiedColor=<rgba 0 0 0 0>}
       cornerRadii=<0 0 0 0>
      }
      tintColor {colorUseRule=UseColor_Specified, specifiedColor=<rgba 0.0091 0.3515 0.9647 1>}
      uVRegion {bIsValid=false, max=<0 0>, min=<0 0>}
     }
     dropIndicator_Onto {
      bIsDynamicallyLoaded=false
      drawAs=Box
      imageType=FullColor
      mirroring=NoMirror
      resourceName=../../../Engine/Content/Slate/Common/DropZoneIndicator_Onto.png
      resourceObject=null
      tiling=NoTile
      imageSize=<0 0>
      margin=<ltrb 0.25 0.25 0.25 0.25>
      outlineSettings {
       bUseBrushTransparency=false
       roundingType=HalfHeightRadius
       width=0
       color {colorUseRule=UseColor_Specified, specifiedColor=<rgba 0 0 0 0>}
       cornerRadii=<0 0 0 0>
      }
      tintColor {colorUseRule=UseColor_Specified, specifiedColor=<rgba 0.0091 0.3515 0.9647 1>}
      uVRegion {bIsValid=false, max=<0 0>, min=<0 0>}
     }
     evenRowBackgroundBrush {
      bIsDynamicallyLoaded=false
      drawAs=Image
      imageType=NoImage
      mirroring=NoMirror
      resourceName=None
      resourceObject=null
      tiling=NoTile
      imageSize=<0 0>
      margin=<ltrb 0 0 0 0>
      outlineSettings {
       bUseBrushTransparency=false
       roundingType=HalfHeightRadius
       width=0
       color {colorUseRule=UseColor_Specified, specifiedColor=<rgba 0 0 0 0>}
       cornerRadii=<0 0 0 0>
      }
      tintColor {colorUseRule=UseColor_Specified, specifiedColor=<rgba 0.0103 0.0103 0.0103 1>}
      uVRegion {bIsValid=false, max=<0 0>, min=<0 0>}
     }
     evenRowBackgroundHoveredBrush {
      bIsDynamicallyLoaded=false
      drawAs=Image
      imageType=NoImage
      mirroring=NoMirror
      resourceName=None
      resourceObject=null
      tiling=NoTile
      imageSize=<0 0>
      margin=<ltrb 0 0 0 0>
      outlineSettings {
       bUseBrushTransparency=false
       roundingType=HalfHeightRadius
       width=0
       color {colorUseRule=UseColor_Specified, specifiedColor=<rgba 0 0 0 0>}
       cornerRadii=<0 0 0 0>
      }
      tintColor {colorUseRule=UseColor_Specified, specifiedColor=<rgba 0.0176 0.0176 0.0176 1>}
      uVRegion {bIsValid=false, max=<0 0>, min=<0 0>}
     }
     inactiveBrush {
      bIsDynamicallyLoaded=false
      drawAs=Image
      imageType=NoImage
      mirroring=NoMirror
      resourceName=None
      resourceObject=null
      tiling=NoTile
      imageSize=<0 0>
      margin=<ltrb 0 0 0 0>
      outlineSettings {
       bUseBrushTransparency=false
       roundingType=HalfHeightRadius
       width=0
       color {colorUseRule=UseColor_Specified, specifiedColor=<rgba 0 0 0 0>}
       cornerRadii=<0 0 0 0>
      }
      tintColor {colorUseRule=UseColor_Specified, specifiedColor=<rgba 0.0513 0.0953 0.159 1>}
      uVRegion {bIsValid=false, max=<0 0>, min=<0 0>}
     }
     inactiveHighlightedBrush {
      bIsDynamicallyLoaded=false
      drawAs=Image
      imageType=NoImage
      mirroring=NoMirror
      resourceName=None
      resourceObject=null
      tiling=NoTile
      imageSize=<0 0>
      margin=<ltrb 0 0 0 0>
      outlineSettings {
       bUseBrushTransparency=false
       roundingType=HalfHeightRadius
       width=0
       color {colorUseRule=UseColor_Specified, specifiedColor=<rgba 0 0 0 0>}
       cornerRadii=<0 0 0 0>
      }
      tintColor {colorUseRule=UseColor_Specified, specifiedColor=<rgba 0.0252 0.0319 0.0423 1>}
      uVRegion {bIsValid=false, max=<0 0>, min=<0 0>}
     }
     inactiveHoveredBrush {
      bIsDynamicallyLoaded=false
      drawAs=Image
      imageType=NoImage
      mirroring=NoMirror
      resourceName=None
      resourceObject=null
      tiling=NoTile
      imageSize=<0 0>
      margin=<ltrb 0 0 0 0>
      outlineSettings {
       bUseBrushTransparency=false
       roundingType=HalfHeightRadius
       width=0
       color {colorUseRule=UseColor_Specified, specifiedColor=<rgba 0 0 0 0>}
       cornerRadii=<0 0 0 0>
      }
      tintColor {colorUseRule=UseColor_Specified, specifiedColor=<rgba 0.0513 0.0953 0.159 1>}
      uVRegion {bIsValid=false, max=<0 0>, min=<0 0>}
     }
     oddRowBackgroundBrush {
      bIsDynamicallyLoaded=false
      drawAs=Image
      imageType=NoImage
      mirroring=NoMirror
      resourceName=None
      resourceObject=null
      tiling=NoTile
      imageSize=<0 0>
      margin=<ltrb 0 0 0 0>
      outlineSettings {
       bUseBrushTransparency=false
       roundingType=HalfHeightRadius
       width=0
       color {colorUseRule=UseColor_Specified, specifiedColor=<rgba 0 0 0 0>}
       cornerRadii=<0 0 0 0>
      }
      tintColor {colorUseRule=UseColor_Specified, specifiedColor=<rgba 0.0103 0.0103 0.0103 1>}
      uVRegion {bIsValid=false, max=<0 0>, min=<0 0>}
     }
     oddRowBackgroundHoveredBrush {
      bIsDynamicallyLoaded=false
      drawAs=Image
      imageType=NoImage
      mirroring=NoMirror
      resourceName=None
      resourceObject=null
      tiling=NoTile
      imageSize=<0 0>
      margin=<ltrb 0 0 0 0>
      outlineSettings {
       bUseBrushTransparency=false
       roundingType=HalfHeightRadius
       width=0
       color {colorUseRule=UseColor_Specified, specifiedColor=<rgba 0 0 0 0>}
       cornerRadii=<0 0 0 0>
      }
      tintColor {colorUseRule=UseColor_Specified, specifiedColor=<rgba 0.0176 0.0176 0.0176 1>}
      uVRegion {bIsValid=false, max=<0 0>, min=<0 0>}
     }
     parentRowBackgroundBrush {
      bIsDynamicallyLoaded=false
      drawAs=Image
      imageType=NoImage
      mirroring=NoMirror
      resourceName=None
      resourceObject=null
      tiling=NoTile
      imageSize=<32 32>
      margin=<ltrb 0 0 0 0>
      outlineSettings {
       bUseBrushTransparency=false
       roundingType=HalfHeightRadius
       width=0
       color {colorUseRule=UseColor_Specified, specifiedColor=<rgba 0 0 0 0>}
       cornerRadii=<0 0 0 0>
      }
      tintColor {colorUseRule=UseColor_Specified, specifiedColor=<rgba 1 1 1 1>}
      uVRegion {bIsValid=false, max=<0 0>, min=<0 0>}
     }
     parentRowBackgroundHoveredBrush {
      bIsDynamicallyLoaded=false
      drawAs=Image
      imageType=NoImage
      mirroring=NoMirror
      resourceName=None
      resourceObject=null
      tiling=NoTile
      imageSize=<32 32>
      margin=<ltrb 0 0 0 0>
      outlineSettings {
       bUseBrushTransparency=false
       roundingType=HalfHeightRadius
       width=0
       color {colorUseRule=UseColor_Specified, specifiedColor=<rgba 0 0 0 0>}
       cornerRadii=<0 0 0 0>
      }
      tintColor {colorUseRule=UseColor_Specified, specifiedColor=<rgba 1 1 1 1>}
      uVRegion {bIsValid=false, max=<0 0>, min=<0 0>}
     }
     selectedTextColor {colorUseRule=UseColor_Specified, specifiedColor=<rgba 1 1 1 1>}
     selectorFocusedBrush {
      bIsDynamicallyLoaded=false
      drawAs=Border
      imageType=FullColor
      mirroring=NoMirror
      resourceName=../../../Engine/Content/Slate/Common/Selector.png
      resourceObject=null
      tiling=Both
      imageSize=<0 0>
      margin=<ltrb 0.25 0.25 0.25 0.25>
      outlineSettings {
       bUseBrushTransparency=false
       roundingType=HalfHeightRadius
       width=0
       color {colorUseRule=UseColor_Specified, specifiedColor=<rgba 0 0 0 0>}
       cornerRadii=<0 0 0 0>
      }
      tintColor {colorUseRule=UseColor_Specified, specifiedColor=<rgba 0.701 0.225 0.003 1>}
      uVRegion {bIsValid=false, max=<0 0>, min=<0 0>}
     }
     textColor {colorUseRule=UseColor_Specified, specifiedColor=<rgba 0.5271 0.5271 0.5271 1>}
    }
    widgetStyle {
     comboButtonStyle {
      downArrowAlign=VAlign_Center
      buttonStyle {
       clickedSlateSound {resourceObject=null}
       disabled {
        bIsDynamicallyLoaded=false
        drawAs=RoundedBox
        imageType=NoImage
        mirroring=NoMirror
        resourceName=None
        resourceObject=null
        tiling=NoTile
        imageSize=<0 0>
        margin=<ltrb 0 0 0 0>
        outlineSettings {
         bUseBrushTransparency=false
         roundingType=FixedRadius
         width=1
         color {colorUseRule=UseColor_ColorTable, specifiedColor=<rgba 0 0 0 0>}
         cornerRadii=<4 4 4 4>
        }
        tintColor {colorUseRule=UseColor_Specified, specifiedColor=<rgba 0.0395 0.0395 0.0395 1>}
        uVRegion {bIsValid=false, max=<0 0>, min=<0 0>}
       }
       disabledForeground {colorUseRule=UseColor_ColorTable, specifiedColor=<rgba 0 0 0 0>}
       hovered {
        bIsDynamicallyLoaded=false
        drawAs=RoundedBox
        imageType=NoImage
        mirroring=NoMirror
        resourceName=None
        resourceObject=null
        tiling=NoTile
        imageSize=<0 0>
        margin=<ltrb 0 0 0 0>
        outlineSettings {
         bUseBrushTransparency=false
         roundingType=FixedRadius
         width=1
         color {colorUseRule=UseColor_ColorTable, specifiedColor=<rgba 0 0 0 0>}
         cornerRadii=<4 4 4 4>
        }
        tintColor {colorUseRule=UseColor_Specified, specifiedColor=<rgba 0.0048 0.0048 0.0048 1>}
        uVRegion {bIsValid=false, max=<0 0>, min=<0 0>}
       }
       hoveredForeground {colorUseRule=UseColor_ColorTable, specifiedColor=<rgba 0 0 0 0>}
       hoveredSlateSound {resourceObject=null}
       normal {
        bIsDynamicallyLoaded=false
        drawAs=RoundedBox
        imageType=NoImage
        mirroring=NoMirror
        resourceName=None
        resourceObject=null
        tiling=NoTile
        imageSize=<0 0>
        margin=<ltrb 0 0 0 0>
        outlineSettings {
         bUseBrushTransparency=false
         roundingType=FixedRadius
         width=1
         color {colorUseRule=UseColor_ColorTable, specifiedColor=<rgba 0 0 0 0>}
         cornerRadii=<4 4 4 4>
        }
        tintColor {colorUseRule=UseColor_Specified, specifiedColor=<rgba 0.0048 0.0048 0.0048 1>}
        uVRegion {bIsValid=false, max=<0 0>, min=<0 0>}
       }
       normalForeground {colorUseRule=UseColor_ColorTable, specifiedColor=<rgba 0 0 0 0>}
       normalPadding=<ltrb 8 1 8 1>
       pressed {
        bIsDynamicallyLoaded=false
        drawAs=RoundedBox
        imageType=NoImage
        mirroring=NoMirror
        resourceName=None
        resourceObject=null
        tiling=NoTile
        imageSize=<0 0>
        margin=<ltrb 0 0 0 0>
        outlineSettings {
         bUseBrushTransparency=false
         roundingType=FixedRadius
         width=1
         color {colorUseRule=UseColor_ColorTable, specifiedColor=<rgba 0 0 0 0>}
         cornerRadii=<4 4 4 4>
        }
        tintColor {colorUseRule=UseColor_Specified, specifiedColor=<rgba 0.0048 0.0048 0.0048 1>}
        uVRegion {bIsValid=false, max=<0 0>, min=<0 0>}
       }
       pressedForeground {colorUseRule=UseColor_ColorTable, specifiedColor=<rgba 0 0 0 0>}
       pressedPadding=<ltrb 8 2 8 0>
       pressedSlateSound {resourceObject=null}
      }
      contentPadding=<ltrb 0 0 0 0>
      downArrowImage {
       bIsDynamicallyLoaded=false
       drawAs=Image
       imageType=Vector
       mirroring=NoMirror
       resourceName=../../../Engine/Content/Slate/Starship/CoreWidgets/ComboBox/wide-chevron-down.svg
       resourceObject=null
       tiling=NoTile
       imageSize=<16 16>
       margin=<ltrb 0 0 0 0>
       outlineSettings {
        bUseBrushTransparency=false
        roundingType=HalfHeightRadius
        width=0
        color {colorUseRule=UseColor_Specified, specifiedColor=<rgba 0 0 0 0>}
        cornerRadii=<0 0 0 0>
       }
       tintColor {colorUseRule=UseColor_Specified, specifiedColor=<rgba 1 1 1 1>}
       uVRegion {bIsValid=false, max=<0 0>, min=<0 0>}
      }
      downArrowPadding=<ltrb 2 2 2 2>
      menuBorderBrush {
       bIsDynamicallyLoaded=false
       drawAs=RoundedBox
       imageType=NoImage
       mirroring=NoMirror
       resourceName=None
       resourceObject=null
       tiling=NoTile
       imageSize=<0 0>
       margin=<ltrb 0 0 0 0>
       outlineSettings {
        bUseBrushTransparency=false
        roundingType=FixedRadius
        width=1
        color {colorUseRule=UseColor_ColorTable, specifiedColor=<rgba 0 0 0 0>}
        cornerRadii=<0 0 0 0>
       }
       tintColor {colorUseRule=UseColor_Specified, specifiedColor=<rgba 0.0395 0.0395 0.0395 1>}
       uVRegion {bIsValid=false, max=<0 0>, min=<0 0>}
      }
      menuBorderPadding=<ltrb 0 0 0 0>
      shadowColorAndOpacity=<rgba 0 0 0 1>
      shadowOffset=<0 0>
     }
     contentPadding=<ltrb 0 0 0 0>
     menuRowPadding=<ltrb 8 3 8 3>
     pressedSlateSound {resourceObject=null}
     selectionChangeSlateSound {resourceObject=null}
    }
   }
   slot {class=VerticalBoxSlot}
  }
  TextBlock {
   widgetClassPath=&TextBlock
   parent=VerticalBox_0
   widget=TextBlock
   properties {text=Give PickAxe}
   slot {class=VerticalBoxSlot, properties {horizontalAlignment=HAlign_Center}}
  }
  SearchableStringComboBox {
   widgetClassPath=&SearchableStringComboBox
   bIsVariable=true
   parent=VerticalBox_0
   widget=SearchableStringComboBox
   properties {
    defaultOptions [
     WID_Athena_FloppingRabbit
     Weapon_Pickaxe_MistMylar
     WID_Athena_ElectroBat
     WID_Harvest_Pickaxe_HightowerTapas1H
     WID_Harvest_Pickaxe_AncientGladiatorMale1H
    ]
    foregroundColor {colorUseRule=UseColor_Specified, specifiedColor=<rgba 0.5271 0.5271 0.5271 1>}
    itemStyle {
     bUseParentRowBrush=false
     activeBrush {
      bIsDynamicallyLoaded=false
      drawAs=Image
      imageType=NoImage
      mirroring=NoMirror
      resourceName=None
      resourceObject=null
      tiling=NoTile
      imageSize=<0 0>
      margin=<ltrb 0 0 0 0>
      outlineSettings {
       bUseBrushTransparency=false
       roundingType=HalfHeightRadius
       width=0
       color {colorUseRule=UseColor_Specified, specifiedColor=<rgba 0 0 0 0>}
       cornerRadii=<0 0 0 0>
      }
      tintColor {colorUseRule=UseColor_Specified, specifiedColor=<rgba 0 0.162 0.7454 1>}
      uVRegion {bIsValid=false, max=<0 0>, min=<0 0>}
     }
     activeHighlightedBrush {
      bIsDynamicallyLoaded=false
      drawAs=Image
      imageType=NoImage
      mirroring=NoMirror
      resourceName=None
      resourceObject=null
      tiling=NoTile
      imageSize=<0 0>
      margin=<ltrb 0 0 0 0>
      outlineSettings {
       bUseBrushTransparency=false
       roundingType=HalfHeightRadius
       width=0
       color {colorUseRule=UseColor_Specified, specifiedColor=<rgba 0 0 0 0>}
       cornerRadii=<0 0 0 0>
      }
      tintColor {colorUseRule=UseColor_Specified, specifiedColor=<rgba 0.0252 0.0319 0.0423 1>}
      uVRegion {bIsValid=false, max=<0 0>, min=<0 0>}
     }
     activeHoveredBrush {
      bIsDynamicallyLoaded=false
      drawAs=Image
      imageType=NoImage
      mirroring=NoMirror
      resourceName=None
      resourceObject=null
      tiling=NoTile
      imageSize=<0 0>
      margin=<ltrb 0 0 0 0>
      outlineSettings {
       bUseBrushTransparency=false
       roundingType=HalfHeightRadius
       width=0
       color {colorUseRule=UseColor_Specified, specifiedColor=<rgba 0 0 0 0>}
       cornerRadii=<0 0 0 0>
      }
      tintColor {colorUseRule=UseColor_Specified, specifiedColor=<rgba 0 0.162 0.7454 1>}
      uVRegion {bIsValid=false, max=<0 0>, min=<0 0>}
     }
     dropIndicator_Above {
      bIsDynamicallyLoaded=false
      drawAs=Box
      imageType=FullColor
      mirroring=NoMirror
      resourceName=../../../Engine/Content/Slate/Common/DropZoneIndicator_Above.png
      resourceObject=null
      tiling=NoTile
      imageSize=<0 0>
      margin=<ltrb 0.625 0.625 0 0>
      outlineSettings {
       bUseBrushTransparency=false
       roundingType=HalfHeightRadius
       width=0
       color {colorUseRule=UseColor_Specified, specifiedColor=<rgba 0 0 0 0>}
       cornerRadii=<0 0 0 0>
      }
      tintColor {colorUseRule=UseColor_Specified, specifiedColor=<rgba 0.0091 0.3515 0.9647 1>}
      uVRegion {bIsValid=false, max=<0 0>, min=<0 0>}
     }
     dropIndicator_Below {
      bIsDynamicallyLoaded=false
      drawAs=Box
      imageType=FullColor
      mirroring=NoMirror
      resourceName=../../../Engine/Content/Slate/Common/DropZoneIndicator_Below.png
      resourceObject=null
      tiling=NoTile
      imageSize=<0 0>
      margin=<ltrb 0.625 0 0 0.625>
      outlineSettings {
       bUseBrushTransparency=false
       roundingType=HalfHeightRadius
       width=0
       color {colorUseRule=UseColor_Specified, specifiedColor=<rgba 0 0 0 0>}
       cornerRadii=<0 0 0 0>
      }
      tintColor {colorUseRule=UseColor_Specified, specifiedColor=<rgba 0.0091 0.3515 0.9647 1>}
      uVRegion {bIsValid=false, max=<0 0>, min=<0 0>}
     }
     dropIndicator_Onto {
      bIsDynamicallyLoaded=false
      drawAs=Box
      imageType=FullColor
      mirroring=NoMirror
      resourceName=../../../Engine/Content/Slate/Common/DropZoneIndicator_Onto.png
      resourceObject=null
      tiling=NoTile
      imageSize=<0 0>
      margin=<ltrb 0.25 0.25 0.25 0.25>
      outlineSettings {
       bUseBrushTransparency=false
       roundingType=HalfHeightRadius
       width=0
       color {colorUseRule=UseColor_Specified, specifiedColor=<rgba 0 0 0 0>}
       cornerRadii=<0 0 0 0>
      }
      tintColor {colorUseRule=UseColor_Specified, specifiedColor=<rgba 0.0091 0.3515 0.9647 1>}
      uVRegion {bIsValid=false, max=<0 0>, min=<0 0>}
     }
     evenRowBackgroundBrush {
      bIsDynamicallyLoaded=false
      drawAs=Image
      imageType=NoImage
      mirroring=NoMirror
      resourceName=None
      resourceObject=null
      tiling=NoTile
      imageSize=<0 0>
      margin=<ltrb 0 0 0 0>
      outlineSettings {
       bUseBrushTransparency=false
       roundingType=HalfHeightRadius
       width=0
       color {colorUseRule=UseColor_Specified, specifiedColor=<rgba 0 0 0 0>}
       cornerRadii=<0 0 0 0>
      }
      tintColor {colorUseRule=UseColor_Specified, specifiedColor=<rgba 0.0103 0.0103 0.0103 1>}
      uVRegion {bIsValid=false, max=<0 0>, min=<0 0>}
     }
     evenRowBackgroundHoveredBrush {
      bIsDynamicallyLoaded=false
      drawAs=Image
      imageType=NoImage
      mirroring=NoMirror
      resourceName=None
      resourceObject=null
      tiling=NoTile
      imageSize=<0 0>
      margin=<ltrb 0 0 0 0>
      outlineSettings {
       bUseBrushTransparency=false
       roundingType=HalfHeightRadius
       width=0
       color {colorUseRule=UseColor_Specified, specifiedColor=<rgba 0 0 0 0>}
       cornerRadii=<0 0 0 0>
      }
      tintColor {colorUseRule=UseColor_Specified, specifiedColor=<rgba 0.0176 0.0176 0.0176 1>}
      uVRegion {bIsValid=false, max=<0 0>, min=<0 0>}
     }
     inactiveBrush {
      bIsDynamicallyLoaded=false
      drawAs=Image
      imageType=NoImage
      mirroring=NoMirror
      resourceName=None
      resourceObject=null
      tiling=NoTile
      imageSize=<0 0>
      margin=<ltrb 0 0 0 0>
      outlineSettings {
       bUseBrushTransparency=false
       roundingType=HalfHeightRadius
       width=0
       color {colorUseRule=UseColor_Specified, specifiedColor=<rgba 0 0 0 0>}
       cornerRadii=<0 0 0 0>
      }
      tintColor {colorUseRule=UseColor_Specified, specifiedColor=<rgba 0.0513 0.0953 0.159 1>}
      uVRegion {bIsValid=false, max=<0 0>, min=<0 0>}
     }
     inactiveHighlightedBrush {
      bIsDynamicallyLoaded=false
      drawAs=Image
      imageType=NoImage
      mirroring=NoMirror
      resourceName=None
      resourceObject=null
      tiling=NoTile
      imageSize=<0 0>
      margin=<ltrb 0 0 0 0>
      outlineSettings {
       bUseBrushTransparency=false
       roundingType=HalfHeightRadius
       width=0
       color {colorUseRule=UseColor_Specified, specifiedColor=<rgba 0 0 0 0>}
       cornerRadii=<0 0 0 0>
      }
      tintColor {colorUseRule=UseColor_Specified, specifiedColor=<rgba 0.0252 0.0319 0.0423 1>}
      uVRegion {bIsValid=false, max=<0 0>, min=<0 0>}
     }
     inactiveHoveredBrush {
      bIsDynamicallyLoaded=false
      drawAs=Image
      imageType=NoImage
      mirroring=NoMirror
      resourceName=None
      resourceObject=null
      tiling=NoTile
      imageSize=<0 0>
      margin=<ltrb 0 0 0 0>
      outlineSettings {
       bUseBrushTransparency=false
       roundingType=HalfHeightRadius
       width=0
       color {colorUseRule=UseColor_Specified, specifiedColor=<rgba 0 0 0 0>}
       cornerRadii=<0 0 0 0>
      }
      tintColor {colorUseRule=UseColor_Specified, specifiedColor=<rgba 0.0513 0.0953 0.159 1>}
      uVRegion {bIsValid=false, max=<0 0>, min=<0 0>}
     }
     oddRowBackgroundBrush {
      bIsDynamicallyLoaded=false
      drawAs=Image
      imageType=NoImage
      mirroring=NoMirror
      resourceName=None
      resourceObject=null
      tiling=NoTile
      imageSize=<0 0>
      margin=<ltrb 0 0 0 0>
      outlineSettings {
       bUseBrushTransparency=false
       roundingType=HalfHeightRadius
       width=0
       color {colorUseRule=UseColor_Specified, specifiedColor=<rgba 0 0 0 0>}
       cornerRadii=<0 0 0 0>
      }
      tintColor {colorUseRule=UseColor_Specified, specifiedColor=<rgba 0.0103 0.0103 0.0103 1>}
      uVRegion {bIsValid=false, max=<0 0>, min=<0 0>}
     }
     oddRowBackgroundHoveredBrush {
      bIsDynamicallyLoaded=false
      drawAs=Image
      imageType=NoImage
      mirroring=NoMirror
      resourceName=None
      resourceObject=null
      tiling=NoTile
      imageSize=<0 0>
      margin=<ltrb 0 0 0 0>
      outlineSettings {
       bUseBrushTransparency=false
       roundingType=HalfHeightRadius
       width=0
       color {colorUseRule=UseColor_Specified, specifiedColor=<rgba 0 0 0 0>}
       cornerRadii=<0 0 0 0>
      }
      tintColor {colorUseRule=UseColor_Specified, specifiedColor=<rgba 0.0176 0.0176 0.0176 1>}
      uVRegion {bIsValid=false, max=<0 0>, min=<0 0>}
     }
     parentRowBackgroundBrush {
      bIsDynamicallyLoaded=false
      drawAs=Image
      imageType=NoImage
      mirroring=NoMirror
      resourceName=None
      resourceObject=null
      tiling=NoTile
      imageSize=<32 32>
      margin=<ltrb 0 0 0 0>
      outlineSettings {
       bUseBrushTransparency=false
       roundingType=HalfHeightRadius
       width=0
       color {colorUseRule=UseColor_Specified, specifiedColor=<rgba 0 0 0 0>}
       cornerRadii=<0 0 0 0>
      }
      tintColor {colorUseRule=UseColor_Specified, specifiedColor=<rgba 1 1 1 1>}
      uVRegion {bIsValid=false, max=<0 0>, min=<0 0>}
     }
     parentRowBackgroundHoveredBrush {
      bIsDynamicallyLoaded=false
      drawAs=Image
      imageType=NoImage
      mirroring=NoMirror
      resourceName=None
      resourceObject=null
      tiling=NoTile
      imageSize=<32 32>
      margin=<ltrb 0 0 0 0>
      outlineSettings {
       bUseBrushTransparency=false
       roundingType=HalfHeightRadius
       width=0
       color {colorUseRule=UseColor_Specified, specifiedColor=<rgba 0 0 0 0>}
       cornerRadii=<0 0 0 0>
      }
      tintColor {colorUseRule=UseColor_Specified, specifiedColor=<rgba 1 1 1 1>}
      uVRegion {bIsValid=false, max=<0 0>, min=<0 0>}
     }
     selectedTextColor {colorUseRule=UseColor_Specified, specifiedColor=<rgba 1 1 1 1>}
     selectorFocusedBrush {
      bIsDynamicallyLoaded=false
      drawAs=Border
      imageType=FullColor
      mirroring=NoMirror
      resourceName=../../../Engine/Content/Slate/Common/Selector.png
      resourceObject=null
      tiling=Both
      imageSize=<0 0>
      margin=<ltrb 0.25 0.25 0.25 0.25>
      outlineSettings {
       bUseBrushTransparency=false
       roundingType=HalfHeightRadius
       width=0
       color {colorUseRule=UseColor_Specified, specifiedColor=<rgba 0 0 0 0>}
       cornerRadii=<0 0 0 0>
      }
      tintColor {colorUseRule=UseColor_Specified, specifiedColor=<rgba 0.701 0.225 0.003 1>}
      uVRegion {bIsValid=false, max=<0 0>, min=<0 0>}
     }
     textColor {colorUseRule=UseColor_Specified, specifiedColor=<rgba 0.5271 0.5271 0.5271 1>}
    }
    widgetStyle {
     comboButtonStyle {
      downArrowAlign=VAlign_Center
      buttonStyle {
       clickedSlateSound {resourceObject=null}
       disabled {
        bIsDynamicallyLoaded=false
        drawAs=RoundedBox
        imageType=NoImage
        mirroring=NoMirror
        resourceName=None
        resourceObject=null
        tiling=NoTile
        imageSize=<0 0>
        margin=<ltrb 0 0 0 0>
        outlineSettings {
         bUseBrushTransparency=false
         roundingType=FixedRadius
         width=1
         color {colorUseRule=UseColor_ColorTable, specifiedColor=<rgba 0 0 0 0>}
         cornerRadii=<4 4 4 4>
        }
        tintColor {colorUseRule=UseColor_Specified, specifiedColor=<rgba 0.0395 0.0395 0.0395 1>}
        uVRegion {bIsValid=false, max=<0 0>, min=<0 0>}
       }
       disabledForeground {colorUseRule=UseColor_ColorTable, specifiedColor=<rgba 0 0 0 0>}
       hovered {
        bIsDynamicallyLoaded=false
        drawAs=RoundedBox
        imageType=NoImage
        mirroring=NoMirror
        resourceName=None
        resourceObject=null
        tiling=NoTile
        imageSize=<0 0>
        margin=<ltrb 0 0 0 0>
        outlineSettings {
         bUseBrushTransparency=false
         roundingType=FixedRadius
         width=1
         color {colorUseRule=UseColor_ColorTable, specifiedColor=<rgba 0 0 0 0>}
         cornerRadii=<4 4 4 4>
        }
        tintColor {colorUseRule=UseColor_Specified, specifiedColor=<rgba 0.0048 0.0048 0.0048 1>}
        uVRegion {bIsValid=false, max=<0 0>, min=<0 0>}
       }
       hoveredForeground {colorUseRule=UseColor_ColorTable, specifiedColor=<rgba 0 0 0 0>}
       hoveredSlateSound {resourceObject=null}
       normal {
        bIsDynamicallyLoaded=false
        drawAs=RoundedBox
        imageType=NoImage
        mirroring=NoMirror
        resourceName=None
        resourceObject=null
        tiling=NoTile
        imageSize=<0 0>
        margin=<ltrb 0 0 0 0>
        outlineSettings {
         bUseBrushTransparency=false
         roundingType=FixedRadius
         width=1
         color {colorUseRule=UseColor_ColorTable, specifiedColor=<rgba 0 0 0 0>}
         cornerRadii=<4 4 4 4>
        }
        tintColor {colorUseRule=UseColor_Specified, specifiedColor=<rgba 0.0048 0.0048 0.0048 1>}
        uVRegion {bIsValid=false, max=<0 0>, min=<0 0>}
       }
       normalForeground {colorUseRule=UseColor_ColorTable, specifiedColor=<rgba 0 0 0 0>}
       normalPadding=<ltrb 8 1 8 1>
       pressed {
        bIsDynamicallyLoaded=false
        drawAs=RoundedBox
        imageType=NoImage
        mirroring=NoMirror
        resourceName=None
        resourceObject=null
        tiling=NoTile
        imageSize=<0 0>
        margin=<ltrb 0 0 0 0>
        outlineSettings {
         bUseBrushTransparency=false
         roundingType=FixedRadius
         width=1
         color {colorUseRule=UseColor_ColorTable, specifiedColor=<rgba 0 0 0 0>}
         cornerRadii=<4 4 4 4>
        }
        tintColor {colorUseRule=UseColor_Specified, specifiedColor=<rgba 0.0048 0.0048 0.0048 1>}
        uVRegion {bIsValid=false, max=<0 0>, min=<0 0>}
       }
       pressedForeground {colorUseRule=UseColor_ColorTable, specifiedColor=<rgba 0 0 0 0>}
       pressedPadding=<ltrb 8 2 8 0>
       pressedSlateSound {resourceObject=null}
      }
      contentPadding=<ltrb 0 0 0 0>
      downArrowImage {
       bIsDynamicallyLoaded=false
       drawAs=Image
       imageType=Vector
       mirroring=NoMirror
       resourceName=../../../Engine/Content/Slate/Starship/CoreWidgets/ComboBox/wide-chevron-down.svg
       resourceObject=null
       tiling=NoTile
       imageSize=<16 16>
       margin=<ltrb 0 0 0 0>
       outlineSettings {
        bUseBrushTransparency=false
        roundingType=HalfHeightRadius
        width=0
        color {colorUseRule=UseColor_Specified, specifiedColor=<rgba 0 0 0 0>}
        cornerRadii=<0 0 0 0>
       }
       tintColor {colorUseRule=UseColor_Specified, specifiedColor=<rgba 1 1 1 1>}
       uVRegion {bIsValid=false, max=<0 0>, min=<0 0>}
      }
      downArrowPadding=<ltrb 2 2 2 2>
      menuBorderBrush {
       bIsDynamicallyLoaded=false
       drawAs=RoundedBox
       imageType=NoImage
       mirroring=NoMirror
       resourceName=None
       resourceObject=null
       tiling=NoTile
       imageSize=<0 0>
       margin=<ltrb 0 0 0 0>
       outlineSettings {
        bUseBrushTransparency=false
        roundingType=FixedRadius
        width=1
        color {colorUseRule=UseColor_ColorTable, specifiedColor=<rgba 0 0 0 0>}
        cornerRadii=<0 0 0 0>
       }
       tintColor {colorUseRule=UseColor_Specified, specifiedColor=<rgba 0.0395 0.0395 0.0395 1>}
       uVRegion {bIsValid=false, max=<0 0>, min=<0 0>}
      }
      menuBorderPadding=<ltrb 0 0 0 0>
      shadowColorAndOpacity=<rgba 0 0 0 1>
      shadowOffset=<0 0>
     }
     contentPadding=<ltrb 0 0 0 0>
     menuRowPadding=<ltrb 8 3 8 3>
     pressedSlateSound {resourceObject=null}
     selectionChangeSlateSound {resourceObject=null}
    }
   }
   slot {class=VerticalBoxSlot}
  }
  TextBlock_1 {
   widgetClassPath=&TextBlock
   parent=VerticalBox_0
   widget=TextBlock_1
   properties {text=Give Specific Item}
   slot {class=VerticalBoxSlot, properties {horizontalAlignment=HAlign_Center}}
  }
  SearchableStringComboBox_1 {
   widgetClassPath=&SearchableStringComboBox
   bIsVariable=true
   parent=VerticalBox_0
   widget=SearchableStringComboBox_1
   properties {
    defaultOptions [
     /UnstableLiquidGun/Items/UnstableLiquidGun/WID_NitroglycerineGun
     /BallisticShieldGameplay/Gameplay/WID_Paprika_BallisticShield
     /SunRoseChainWhipGameplay/Items/ChainWhip/WID_ChainWhip
     /PrimalGameplay/Items/Bows/Bone/WID_Bow_Bone_Athena_R
     /ChainsawGameplay/Gameplay/AGID_Chainsaw
     /KatanaGameplay/Items/Katana/WID_Melee_Katana
     /ElectroBatGameplay/Gameplay/WID_Athena_ElectroBat
     /RocketDrillGameplay/Gameplay/Items/AGID_RocketDrill_JetpackVer
    ]
    foregroundColor {colorUseRule=UseColor_Specified, specifiedColor=<rgba 0.5271 0.5271 0.5271 1>}
    itemStyle {
     bUseParentRowBrush=false
     activeBrush {
      bIsDynamicallyLoaded=false
      drawAs=Image
      imageType=NoImage
      mirroring=NoMirror
      resourceName=None
      resourceObject=null
      tiling=NoTile
      imageSize=<0 0>
      margin=<ltrb 0 0 0 0>
      outlineSettings {
       bUseBrushTransparency=false
       roundingType=HalfHeightRadius
       width=0
       color {colorUseRule=UseColor_Specified, specifiedColor=<rgba 0 0 0 0>}
       cornerRadii=<0 0 0 0>
      }
      tintColor {colorUseRule=UseColor_Specified, specifiedColor=<rgba 0 0.162 0.7454 1>}
      uVRegion {bIsValid=false, max=<0 0>, min=<0 0>}
     }
     activeHighlightedBrush {
      bIsDynamicallyLoaded=false
      drawAs=Image
      imageType=NoImage
      mirroring=NoMirror
      resourceName=None
      resourceObject=null
      tiling=NoTile
      imageSize=<0 0>
      margin=<ltrb 0 0 0 0>
      outlineSettings {
       bUseBrushTransparency=false
       roundingType=HalfHeightRadius
       width=0
       color {colorUseRule=UseColor_Specified, specifiedColor=<rgba 0 0 0 0>}
       cornerRadii=<0 0 0 0>
      }
      tintColor {colorUseRule=UseColor_Specified, specifiedColor=<rgba 0.0252 0.0319 0.0423 1>}
      uVRegion {bIsValid=false, max=<0 0>, min=<0 0>}
     }
     activeHoveredBrush {
      bIsDynamicallyLoaded=false
      drawAs=Image
      imageType=NoImage
      mirroring=NoMirror
      resourceName=None
      resourceObject=null
      tiling=NoTile
      imageSize=<0 0>
      margin=<ltrb 0 0 0 0>
      outlineSettings {
       bUseBrushTransparency=false
       roundingType=HalfHeightRadius
       width=0
       color {colorUseRule=UseColor_Specified, specifiedColor=<rgba 0 0 0 0>}
       cornerRadii=<0 0 0 0>
      }
      tintColor {colorUseRule=UseColor_Specified, specifiedColor=<rgba 0 0.162 0.7454 1>}
      uVRegion {bIsValid=false, max=<0 0>, min=<0 0>}
     }
     dropIndicator_Above {
      bIsDynamicallyLoaded=false
      drawAs=Box
      imageType=FullColor
      mirroring=NoMirror
      resourceName=../../../Engine/Content/Slate/Common/DropZoneIndicator_Above.png
      resourceObject=null
      tiling=NoTile
      imageSize=<0 0>
      margin=<ltrb 0.625 0.625 0 0>
      outlineSettings {
       bUseBrushTransparency=false
       roundingType=HalfHeightRadius
       width=0
       color {colorUseRule=UseColor_Specified, specifiedColor=<rgba 0 0 0 0>}
       cornerRadii=<0 0 0 0>
      }
      tintColor {colorUseRule=UseColor_Specified, specifiedColor=<rgba 0.0091 0.3515 0.9647 1>}
      uVRegion {bIsValid=false, max=<0 0>, min=<0 0>}
     }
     dropIndicator_Below {
      bIsDynamicallyLoaded=false
      drawAs=Box
      imageType=FullColor
      mirroring=NoMirror
      resourceName=../../../Engine/Content/Slate/Common/DropZoneIndicator_Below.png
      resourceObject=null
      tiling=NoTile
      imageSize=<0 0>
      margin=<ltrb 0.625 0 0 0.625>
      outlineSettings {
       bUseBrushTransparency=false
       roundingType=HalfHeightRadius
       width=0
       color {colorUseRule=UseColor_Specified, specifiedColor=<rgba 0 0 0 0>}
       cornerRadii=<0 0 0 0>
      }
      tintColor {colorUseRule=UseColor_Specified, specifiedColor=<rgba 0.0091 0.3515 0.9647 1>}
      uVRegion {bIsValid=false, max=<0 0>, min=<0 0>}
     }
     dropIndicator_Onto {
      bIsDynamicallyLoaded=false
      drawAs=Box
      imageType=FullColor
      mirroring=NoMirror
      resourceName=../../../Engine/Content/Slate/Common/DropZoneIndicator_Onto.png
      resourceObject=null
      tiling=NoTile
      imageSize=<0 0>
      margin=<ltrb 0.25 0.25 0.25 0.25>
      outlineSettings {
       bUseBrushTransparency=false
       roundingType=HalfHeightRadius
       width=0
       color {colorUseRule=UseColor_Specified, specifiedColor=<rgba 0 0 0 0>}
       cornerRadii=<0 0 0 0>
      }
      tintColor {colorUseRule=UseColor_Specified, specifiedColor=<rgba 0.0091 0.3515 0.9647 1>}
      uVRegion {bIsValid=false, max=<0 0>, min=<0 0>}
     }
     evenRowBackgroundBrush {
      bIsDynamicallyLoaded=false
      drawAs=Image
      imageType=NoImage
      mirroring=NoMirror
      resourceName=None
      resourceObject=null
      tiling=NoTile
      imageSize=<0 0>
      margin=<ltrb 0 0 0 0>
      outlineSettings {
       bUseBrushTransparency=false
       roundingType=HalfHeightRadius
       width=0
       color {colorUseRule=UseColor_Specified, specifiedColor=<rgba 0 0 0 0>}
       cornerRadii=<0 0 0 0>
      }
      tintColor {colorUseRule=UseColor_Specified, specifiedColor=<rgba 0.0103 0.0103 0.0103 1>}
      uVRegion {bIsValid=false, max=<0 0>, min=<0 0>}
     }
     evenRowBackgroundHoveredBrush {
      bIsDynamicallyLoaded=false
      drawAs=Image
      imageType=NoImage
      mirroring=NoMirror
      resourceName=None
      resourceObject=null
      tiling=NoTile
      imageSize=<0 0>
      margin=<ltrb 0 0 0 0>
      outlineSettings {
       bUseBrushTransparency=false
       roundingType=HalfHeightRadius
       width=0
       color {colorUseRule=UseColor_Specified, specifiedColor=<rgba 0 0 0 0>}
       cornerRadii=<0 0 0 0>
      }
      tintColor {colorUseRule=UseColor_Specified, specifiedColor=<rgba 0.0176 0.0176 0.0176 1>}
      uVRegion {bIsValid=false, max=<0 0>, min=<0 0>}
     }
     inactiveBrush {
      bIsDynamicallyLoaded=false
      drawAs=Image
      imageType=NoImage
      mirroring=NoMirror
      resourceName=None
      resourceObject=null
      tiling=NoTile
      imageSize=<0 0>
      margin=<ltrb 0 0 0 0>
      outlineSettings {
       bUseBrushTransparency=false
       roundingType=HalfHeightRadius
       width=0
       color {colorUseRule=UseColor_Specified, specifiedColor=<rgba 0 0 0 0>}
       cornerRadii=<0 0 0 0>
      }
      tintColor {colorUseRule=UseColor_Specified, specifiedColor=<rgba 0.0513 0.0953 0.159 1>}
      uVRegion {bIsValid=false, max=<0 0>, min=<0 0>}
     }
     inactiveHighlightedBrush {
      bIsDynamicallyLoaded=false
      drawAs=Image
      imageType=NoImage
      mirroring=NoMirror
      resourceName=None
      resourceObject=null
      tiling=NoTile
      imageSize=<0 0>
      margin=<ltrb 0 0 0 0>
      outlineSettings {
       bUseBrushTransparency=false
       roundingType=HalfHeightRadius
       width=0
       color {colorUseRule=UseColor_Specified, specifiedColor=<rgba 0 0 0 0>}
       cornerRadii=<0 0 0 0>
      }
      tintColor {colorUseRule=UseColor_Specified, specifiedColor=<rgba 0.0252 0.0319 0.0423 1>}
      uVRegion {bIsValid=false, max=<0 0>, min=<0 0>}
     }
     inactiveHoveredBrush {
      bIsDynamicallyLoaded=false
      drawAs=Image
      imageType=NoImage
      mirroring=NoMirror
      resourceName=None
      resourceObject=null
      tiling=NoTile
      imageSize=<0 0>
      margin=<ltrb 0 0 0 0>
      outlineSettings {
       bUseBrushTransparency=false
       roundingType=HalfHeightRadius
       width=0
       color {colorUseRule=UseColor_Specified, specifiedColor=<rgba 0 0 0 0>}
       cornerRadii=<0 0 0 0>
      }
      tintColor {colorUseRule=UseColor_Specified, specifiedColor=<rgba 0.0513 0.0953 0.159 1>}
      uVRegion {bIsValid=false, max=<0 0>, min=<0 0>}
     }
     oddRowBackgroundBrush {
      bIsDynamicallyLoaded=false
      drawAs=Image
      imageType=NoImage
      mirroring=NoMirror
      resourceName=None
      resourceObject=null
      tiling=NoTile
      imageSize=<0 0>
      margin=<ltrb 0 0 0 0>
      outlineSettings {
       bUseBrushTransparency=false
       roundingType=HalfHeightRadius
       width=0
       color {colorUseRule=UseColor_Specified, specifiedColor=<rgba 0 0 0 0>}
       cornerRadii=<0 0 0 0>
      }
      tintColor {colorUseRule=UseColor_Specified, specifiedColor=<rgba 0.0103 0.0103 0.0103 1>}
      uVRegion {bIsValid=false, max=<0 0>, min=<0 0>}
     }
     oddRowBackgroundHoveredBrush {
      bIsDynamicallyLoaded=false
      drawAs=Image
      imageType=NoImage
      mirroring=NoMirror
      resourceName=None
      resourceObject=null
      tiling=NoTile
      imageSize=<0 0>
      margin=<ltrb 0 0 0 0>
      outlineSettings {
       bUseBrushTransparency=false
       roundingType=HalfHeightRadius
       width=0
       color {colorUseRule=UseColor_Specified, specifiedColor=<rgba 0 0 0 0>}
       cornerRadii=<0 0 0 0>
      }
      tintColor {colorUseRule=UseColor_Specified, specifiedColor=<rgba 0.0176 0.0176 0.0176 1>}
      uVRegion {bIsValid=false, max=<0 0>, min=<0 0>}
     }
     parentRowBackgroundBrush {
      bIsDynamicallyLoaded=false
      drawAs=Image
      imageType=NoImage
      mirroring=NoMirror
      resourceName=None
      resourceObject=null
      tiling=NoTile
      imageSize=<32 32>
      margin=<ltrb 0 0 0 0>
      outlineSettings {
       bUseBrushTransparency=false
       roundingType=HalfHeightRadius
       width=0
       color {colorUseRule=UseColor_Specified, specifiedColor=<rgba 0 0 0 0>}
       cornerRadii=<0 0 0 0>
      }
      tintColor {colorUseRule=UseColor_Specified, specifiedColor=<rgba 1 1 1 1>}
      uVRegion {bIsValid=false, max=<0 0>, min=<0 0>}
     }
     parentRowBackgroundHoveredBrush {
      bIsDynamicallyLoaded=false
      drawAs=Image
      imageType=NoImage
      mirroring=NoMirror
      resourceName=None
      resourceObject=null
      tiling=NoTile
      imageSize=<32 32>
      margin=<ltrb 0 0 0 0>
      outlineSettings {
       bUseBrushTransparency=false
       roundingType=HalfHeightRadius
       width=0
       color {colorUseRule=UseColor_Specified, specifiedColor=<rgba 0 0 0 0>}
       cornerRadii=<0 0 0 0>
      }
      tintColor {colorUseRule=UseColor_Specified, specifiedColor=<rgba 1 1 1 1>}
      uVRegion {bIsValid=false, max=<0 0>, min=<0 0>}
     }
     selectedTextColor {colorUseRule=UseColor_Specified, specifiedColor=<rgba 1 1 1 1>}
     selectorFocusedBrush {
      bIsDynamicallyLoaded=false
      drawAs=Border
      imageType=FullColor
      mirroring=NoMirror
      resourceName=../../../Engine/Content/Slate/Common/Selector.png
      resourceObject=null
      tiling=Both
      imageSize=<0 0>
      margin=<ltrb 0.25 0.25 0.25 0.25>
      outlineSettings {
       bUseBrushTransparency=false
       roundingType=HalfHeightRadius
       width=0
       color {colorUseRule=UseColor_Specified, specifiedColor=<rgba 0 0 0 0>}
       cornerRadii=<0 0 0 0>
      }
      tintColor {colorUseRule=UseColor_Specified, specifiedColor=<rgba 0.701 0.225 0.003 1>}
      uVRegion {bIsValid=false, max=<0 0>, min=<0 0>}
     }
     textColor {colorUseRule=UseColor_Specified, specifiedColor=<rgba 0.5271 0.5271 0.5271 1>}
    }
    widgetStyle {
     comboButtonStyle {
      downArrowAlign=VAlign_Center
      buttonStyle {
       clickedSlateSound {resourceObject=null}
       disabled {
        bIsDynamicallyLoaded=false
        drawAs=RoundedBox
        imageType=NoImage
        mirroring=NoMirror
        resourceName=None
        resourceObject=null
        tiling=NoTile
        imageSize=<0 0>
        margin=<ltrb 0 0 0 0>
        outlineSettings {
         bUseBrushTransparency=false
         roundingType=FixedRadius
         width=1
         color {colorUseRule=UseColor_ColorTable, specifiedColor=<rgba 0 0 0 0>}
         cornerRadii=<4 4 4 4>
        }
        tintColor {colorUseRule=UseColor_Specified, specifiedColor=<rgba 0.0395 0.0395 0.0395 1>}
        uVRegion {bIsValid=false, max=<0 0>, min=<0 0>}
       }
       disabledForeground {colorUseRule=UseColor_ColorTable, specifiedColor=<rgba 0 0 0 0>}
       hovered {
        bIsDynamicallyLoaded=false
        drawAs=RoundedBox
        imageType=NoImage
        mirroring=NoMirror
        resourceName=None
        resourceObject=null
        tiling=NoTile
        imageSize=<0 0>
        margin=<ltrb 0 0 0 0>
        outlineSettings {
         bUseBrushTransparency=false
         roundingType=FixedRadius
         width=1
         color {colorUseRule=UseColor_ColorTable, specifiedColor=<rgba 0 0 0 0>}
         cornerRadii=<4 4 4 4>
        }
        tintColor {colorUseRule=UseColor_Specified, specifiedColor=<rgba 0.0048 0.0048 0.0048 1>}
        uVRegion {bIsValid=false, max=<0 0>, min=<0 0>}
       }
       hoveredForeground {colorUseRule=UseColor_ColorTable, specifiedColor=<rgba 0 0 0 0>}
       hoveredSlateSound {resourceObject=null}
       normal {
        bIsDynamicallyLoaded=false
        drawAs=RoundedBox
        imageType=NoImage
        mirroring=NoMirror
        resourceName=None
        resourceObject=null
        tiling=NoTile
        imageSize=<0 0>
        margin=<ltrb 0 0 0 0>
        outlineSettings {
         bUseBrushTransparency=false
         roundingType=FixedRadius
         width=1
         color {colorUseRule=UseColor_ColorTable, specifiedColor=<rgba 0 0 0 0>}
         cornerRadii=<4 4 4 4>
        }
        tintColor {colorUseRule=UseColor_Specified, specifiedColor=<rgba 0.0048 0.0048 0.0048 1>}
        uVRegion {bIsValid=false, max=<0 0>, min=<0 0>}
       }
       normalForeground {colorUseRule=UseColor_ColorTable, specifiedColor=<rgba 0 0 0 0>}
       normalPadding=<ltrb 8 1 8 1>
       pressed {
        bIsDynamicallyLoaded=false
        drawAs=RoundedBox
        imageType=NoImage
        mirroring=NoMirror
        resourceName=None
        resourceObject=null
        tiling=NoTile
        imageSize=<0 0>
        margin=<ltrb 0 0 0 0>
        outlineSettings {
         bUseBrushTransparency=false
         roundingType=FixedRadius
         width=1
         color {colorUseRule=UseColor_ColorTable, specifiedColor=<rgba 0 0 0 0>}
         cornerRadii=<4 4 4 4>
        }
        tintColor {colorUseRule=UseColor_Specified, specifiedColor=<rgba 0.0048 0.0048 0.0048 1>}
        uVRegion {bIsValid=false, max=<0 0>, min=<0 0>}
       }
       pressedForeground {colorUseRule=UseColor_ColorTable, specifiedColor=<rgba 0 0 0 0>}
       pressedPadding=<ltrb 8 2 8 0>
       pressedSlateSound {resourceObject=null}
      }
      contentPadding=<ltrb 0 0 0 0>
      downArrowImage {
       bIsDynamicallyLoaded=false
       drawAs=Image
       imageType=Vector
       mirroring=NoMirror
       resourceName=../../../Engine/Content/Slate/Starship/CoreWidgets/ComboBox/wide-chevron-down.svg
       resourceObject=null
       tiling=NoTile
       imageSize=<16 16>
       margin=<ltrb 0 0 0 0>
       outlineSettings {
        bUseBrushTransparency=false
        roundingType=HalfHeightRadius
        width=0
        color {colorUseRule=UseColor_Specified, specifiedColor=<rgba 0 0 0 0>}
        cornerRadii=<0 0 0 0>
       }
       tintColor {colorUseRule=UseColor_Specified, specifiedColor=<rgba 1 1 1 1>}
       uVRegion {bIsValid=false, max=<0 0>, min=<0 0>}
      }
      downArrowPadding=<ltrb 2 2 2 2>
      menuBorderBrush {
       bIsDynamicallyLoaded=false
       drawAs=RoundedBox
       imageType=NoImage
       mirroring=NoMirror
       resourceName=None
       resourceObject=null
       tiling=NoTile
       imageSize=<0 0>
       margin=<ltrb 0 0 0 0>
       outlineSettings {
        bUseBrushTransparency=false
        roundingType=FixedRadius
        width=1
        color {colorUseRule=UseColor_ColorTable, specifiedColor=<rgba 0 0 0 0>}
        cornerRadii=<0 0 0 0>
       }
       tintColor {colorUseRule=UseColor_Specified, specifiedColor=<rgba 0.0395 0.0395 0.0395 1>}
       uVRegion {bIsValid=false, max=<0 0>, min=<0 0>}
      }
      menuBorderPadding=<ltrb 0 0 0 0>
      shadowColorAndOpacity=<rgba 0 0 0 1>
      shadowOffset=<0 0>
     }
     contentPadding=<ltrb 0 0 0 0>
     menuRowPadding=<ltrb 8 3 8 3>
     pressedSlateSound {resourceObject=null}
     selectionChangeSlateSound {resourceObject=null}
    }
   }
   slot {class=VerticalBoxSlot}
  }
 ]
}
