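# Editor utility widget that inspects how the player's current weapon is attached during Play In Editor.
# Its button clears stored values, then takes the first player pawn from the first PIE world and reads its current weapon.
# It records the weapon mesh's attach parent and socket, both hand socket name overrides, and the first socket part override.
# Two disabled details views bound to the widget itself display these variables by category, serving as a read-only readout.
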EditorUtilityWidgetBlueprint /BRCosmetics/Characters/Player/Male/Large/Bodies/M_LRG_ChimneyStack/IK_Retarget_New/EUW_WeaponAttachmentInspector
graphs [
 EventGraph {nodes [
   "On Clicked (EditorUtilityButton_25)"() {is=title K2Node_ComponentBoundEvent}
   > Set FortPlayerPawn(FortPlayerPawn:Fort Player Pawn Object Reference) -> (Output_Get:Fort Player Pawn Object Reference)
   > Set FortWeapon(FortWeapon:Fort Weapon Object Reference) -> (Output_Get:Fort Weapon Object Reference)
   > Set ParentComponent(ParentComponent:Scene Component Object Reference) -> (Output_Get:Scene Component Object Reference)
   > Set SocketName(SocketName:Name=None) -> (Output_Get:Name=None)
   > Set WeaponHandSocketNameOverride(WeaponHandSocketNameOverride:Name=None) -> (Output_Get:Name=None)
   > Set LeftHandWeaponHandSocketNameOverride(LeftHandWeaponHandSocketNameOverride:Name=None) -> (Output_Get:Name=None)
   > Set WeaponHandSocketPartOverride(WeaponHandSocketPartOverride:Name=None) -> (Output_Get:Name=None)
   > n1 = GetFortPlayerPawn!(execute, self) -> (then, OutFortPlayerPawn:Fort Player Pawn Object Reference)
   > Set FortPlayerPawn(FortPlayerPawn=@n1)
   n2 = Get FortPlayerPawn() -> (FortPlayerPawn:Fort Player Pawn Object Reference)
   > Is Valid(InputObject=@n2)
   > Set FortWeapon(FortWeapon=GetCurrentWeapon!(self:Fort Pawn Object Reference=@n2) -> (ReturnValue:Fort Weapon Object Reference))
   > Is Valid(InputObject=Get FortWeapon() -> (FortWeapon:Fort Weapon Object Reference))
   > n3 = Sequence()
   n4 = Get WeaponMesh!(self:Fort Weapon Object Reference=Get FortWeapon()) -> (WeaponMesh:Skeletal Mesh Component Budgeted Object Reference)
   > Set ParentComponent(ParentComponent=GetAttachParent!(self:Scene Component Object Reference=@n4) -> (ReturnValue:Scene Component Object Reference))
   > Set SocketName(SocketName=GetAttachSocketName!(self:Scene Component Object Reference=@n4) -> (ReturnValue:Name=None)) -> (Output_Get=None)
   Set LeftHandWeaponHandSocketNameOverride(
    execute=@n3.then_1
    LeftHandWeaponHandSocketNameOverride=Get LeftHandWeaponHandSocketNameOverride!(self:Fort Weapon Object Reference=Get FortWeapon()) -> (LeftHandWeaponHandSocketNameOverride:Name=None)
   ) -> (Output_Get=None)
   > Set WeaponHandSocketNameOverride(WeaponHandSocketNameOverride=Get WeaponHandSocketNameOverride!(self:Fort Weapon Object Reference=Get FortWeapon()) -> (WeaponHandSocketNameOverride:Name=None)) -> (Output_Get=None)
   n5 = Get WeaponHandSocketPartOverrides!(self:Fort Weapon Object Reference=Get FortWeapon()) -> (WeaponHandSocketPartOverrides:Array of EFortCustomPartType Enums=Head)
   > Branch(Condition=Is Not Empty!(TargetArray:Array of EFortCustomPartType Enums=@n5) -> (ReturnValue:bool))
   > Set WeaponHandSocketPartOverride(WeaponHandSocketPartOverride=Enum to Name!(Enumerator:EFortCustomPartType Enum="Get (a copy)"(Array:Array of EFortCustomPartType Enums=@n5) -> (Output:EFortCustomPartType Enum)) -> (ReturnValue:Name) {is=title K2Node_GetEnumeratorName}) -> (Output_Get=None)
   Event PreConstruct()
   > SetObject(
    self=[
     Get AttachmentOverride_DetailsView() -> (AttachmentOverride_DetailsView:Details View Object Reference)
     Get Weapon_DetailsView() -> (Weapon_DetailsView:Details View Object Reference)
    ]
    NewObject=Self-Reference()
   )
   Event Tick()
  ]}
 GetFortPlayerPawn {nodes [
   GetFortPlayerPawn() {is=title K2Node_FunctionEntry}
   > n1 = GetPIEWorlds!(execute, bIncludeDedicatedServer:bool=false) -> (then, ReturnValue:Array of World Object References)
   > Branch(Condition=Is Not Empty(TargetArray:Array of World Object References=@n1))
   > n2 = GetFortPlayerPawns!(execute, WorldContextObject:Object Reference="Get (a copy)"(Array:Array of World Object References=@n1) -> (Output:World Object Reference)) -> (then, OutFortPlayerPawns:Array of Fort Player Pawn Object References)
   > Branch(Condition=Is Not Empty(TargetArray:Array of Fort Player Pawn Object References=@n2))
   > Return Node!(
    execute
    OutFortPlayerPawn:Fort Player Pawn Object Reference="Get (a copy)"(Array:Array of Fort Player Pawn Object References=@n2) -> (Output:Fort Player Pawn Object Reference)
   )
  ]}
]
variables [
 FortPlayerPawn {type=FortPlayerPawn}
 FortWeapon {type=FortWeapon, default=null}
 LeftHandWeaponHandSocketNameOverride {type=string, default=None}
 ParentComponent {type=SceneComponent, default=null}
 SocketName {type=string, default=None}
 WeaponHandSocketNameOverride {type=string, default=None}
 WeaponHandSocketPartOverride {type=string, default=None}
]
widgets {
 info {rootWidgetClass=&VerticalBox, widgetCount=5}
 widgets [
  VerticalBox_0 {widgetClassPath=&VerticalBox, widget=VerticalBox_0, properties {bOverride_Cursor=true}}
  Weapon_DetailsView {
   widgetClassPath=&DetailsView
   parent=VerticalBox_0
   widget=Weapon_DetailsView
   properties {bIsEnabled=false, categoriesToShow [Weapon]}
   slot {class=VerticalBoxSlot, properties {padding=<ltrb 5 5 5 5>}}
  }
  AttachmentOverride_DetailsView {
   widgetClassPath=&DetailsView
   parent=VerticalBox_0
   widget=AttachmentOverride_DetailsView
   properties {bIsEnabled=false, categoriesToShow [AttachmentOverride]}
   slot {class=VerticalBoxSlot, properties {padding=<ltrb 5 5 5 5>}}
  }
  EditorUtilityButton_25 {widgetClassPath=&EditorUtilityButton, parent=VerticalBox_0, widget=EditorUtilityButton_25, slot {class=VerticalBoxSlot}}
  TextBlock_33 {
   widgetClassPath=&TextBlock
   parent=EditorUtilityButton_25
   widget=TextBlock_33
   properties {text=Inspect Weapon Attachment}
   slot {class=ButtonSlot}
  }
 ]
}
