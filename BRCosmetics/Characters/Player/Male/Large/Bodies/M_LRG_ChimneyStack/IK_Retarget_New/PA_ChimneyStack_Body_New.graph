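# Physics asset for the body of the large male player character ChimneyStack, in its IK_Retarget_New folder.
# One capsule per body covers pelvis, spine, neck, clavicles, upper arms, hands, thighs and calves; forearms, head and feet have no bodies.
# Constraints form a tree from the pelvis, with each hand linked directly to its upper arm.
# Every joint shares the same 45-degree swing and twist limit values rather than per-joint anatomical tuning.
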PhysicsAsset /BRCosmetics/Characters/Player/Male/Large/Bodies/M_LRG_ChimneyStack/IK_Retarget_New/PA_ChimneyStack_Body_New
bodies {
 calf_l [{shapeType=Capsule, length=12.0976, radius=12.0522, center=<-14.2612 1.1762 0.1744>, rotation=<rot -1.2881 85.3744 -88.1721>}]
 calf_r [{shapeType=Capsule, length=12.0976, radius=12.0522, center=<14.2612 -1.1762 -0.1744>, rotation=<rot -1.2881 85.3744 91.8279>}]
 clavicle_l [{shapeType=Capsule, length=17.9401, radius=9.2854, center=<9.1504 3.4073 -2.7762>, rotation=<rot 23.4126 -71.9062 -84.8975>}]
 clavicle_r [{shapeType=Capsule, length=17.9401, radius=9.2854, center=<-9.1504 -3.4073 2.7761>, rotation=<rot 23.4126 -71.9062 -84.8975>}]
 hand_l [{shapeType=Capsule, length=9.4785, radius=5.805, center=<5.6885 -2.2953 -0.2815>, rotation=<rot 0 97.5083 -79.2567>}]
 hand_r [{shapeType=Capsule, length=9.4785, radius=5.805, center=<-5.6885 2.2953 0.2814>, rotation=<rot 0 97.5083 -79.2566>}]
 neck_02 [{shapeType=Capsule, length=20.652, radius=14.7169, center=<8.3353 -4.2713 -0.0007>, rotation=<rot 0 -161.4249 -89.9902>}]
 pelvis [{shapeType=Capsule, length=29.9514, radius=30.1375, center=<-7.1224 -0.6811 0>, rotation=<rot 0 -90 0>}]
 spine_02 [{shapeType=Capsule, length=32.7685, radius=28.4485, center=<0.7881 -2.4785 0>, rotation=<rot 0 -90 0>}]
 spine_03 [{shapeType=Capsule, length=25.6559, radius=19.8923, center=<13.0842 1.1516 0>, rotation=<rot 0 5.9669 0.9651>}]
 spine_04 [{shapeType=Capsule, length=31.0601, radius=23.4199, center=<-5.1359 -2.0708 0>, rotation=<rot 0 -90 0>}]
 spine_05 [{shapeType=Capsule, length=18.1823, radius=15.0185, center=<-3.1144 -5.8081 0.9979>, rotation=<rot 0 -169.3593 -10.6577>}]
 thigh_l [{shapeType=Capsule, length=18.9054, radius=16.0405, center=<-6.0705 -1.2537 0.6258>, rotation=<rot 90 0 0>}]
 thigh_r [{shapeType=Capsule, length=18.9054, radius=16.0405, center=<6.0705 1.2537 -0.6258>, rotation=<rot 90 180 0>}]
 upperarm_l [{shapeType=Capsule, length=33.3904, radius=11.7685, center=<26.0932 4.5576 -2.8754>, rotation=<rot 0 105.8582 -88.6973>}]
 upperarm_r [{shapeType=Capsule, length=32.8371, radius=12.3219, center=<-26.0932 -4.5576 2.8754>, rotation=<rot 0 105.8582 -88.6973>}]
}
constraints [
 {bone1Name=thigh_r, bone2Name=pelvis, swing1LimitDegrees=45, swing2LimitDegrees=45, twistLimitDegrees=45}
 {bone1Name=calf_r, bone2Name=thigh_r, swing1LimitDegrees=45, swing2LimitDegrees=45, twistLimitDegrees=45}
 {bone1Name=thigh_l, bone2Name=pelvis, swing1LimitDegrees=45, swing2LimitDegrees=45, twistLimitDegrees=45}
 {bone1Name=calf_l, bone2Name=thigh_l, swing1LimitDegrees=45, swing2LimitDegrees=45, twistLimitDegrees=45}
 {bone1Name=spine_02, bone2Name=pelvis, swing1LimitDegrees=45, swing2LimitDegrees=45, twistLimitDegrees=45}
 {bone1Name=spine_03, bone2Name=spine_02, swing1LimitDegrees=45, swing2LimitDegrees=45, twistLimitDegrees=45}
 {bone1Name=spine_04, bone2Name=spine_03, swing1LimitDegrees=45, swing2LimitDegrees=45, twistLimitDegrees=45}
 {bone1Name=spine_05, bone2Name=spine_04, swing1LimitDegrees=45, swing2LimitDegrees=45, twistLimitDegrees=45}
 {bone1Name=neck_02, bone2Name=spine_05, swing1LimitDegrees=45, swing2LimitDegrees=45, twistLimitDegrees=45}
 {bone1Name=clavicle_l, bone2Name=spine_05, swing1LimitDegrees=45, swing2LimitDegrees=45, twistLimitDegrees=45}
 {bone1Name=upperarm_l, bone2Name=clavicle_l, swing1LimitDegrees=45, swing2LimitDegrees=45, twistLimitDegrees=45}
 {bone1Name=hand_l, bone2Name=upperarm_l, swing1LimitDegrees=45, swing2LimitDegrees=45, twistLimitDegrees=45}
 {bone1Name=clavicle_r, bone2Name=spine_05, swing1LimitDegrees=45, swing2LimitDegrees=45, twistLimitDegrees=45}
 {bone1Name=upperarm_r, bone2Name=clavicle_r, swing1LimitDegrees=45, swing2LimitDegrees=45, twistLimitDegrees=45}
 {bone1Name=hand_r, bone2Name=upperarm_r, swing1LimitDegrees=45, swing2LimitDegrees=45, twistLimitDegrees=45}
]
