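# Physics asset for the ChimneyStack large male player body outfit, kept in its new IK retarget folder.
# Single capsules approximate pelvis, spine, head, arms, hands, legs and feet, with a tiny unrotated placeholder on neck_02.
# Constraints chain the bodies out from the pelvis, with free swing and twist at the head, both elbows and left ankle.
# The right foot has a body but no constraint, and the head and thigh capsules have negative lengths.
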PhysicsAsset /BRCosmetics/Characters/Player/Male/Large/Bodies/M_LRG_ChimneyStack/IK_Retarget_New/PA_Outfit_ChimneyStack_Relative_Base
bodies {
 calf_l [{
   shapeType=Capsule
   length=14.7236
   radius=10.9214
   shapeName=calf_l_capsule
   center=<-8.6885 0.4427 1.3472>
   rotation=<rot -78.2285 -97.5483 106.7323>
  }]
 calf_r [{
   shapeType=Capsule
   length=14.7867
   radius=10.9835
   shapeName=calf_r_capsule
   center=<8.4752 -0.4591 -1.3429>
   rotation=<rot -76.8761 -97.6821 -72.5286>
  }]
 foot_l [{
   shapeType=Capsule
   length=9.3872
   radius=5.5994
   shapeName=foot_l_capsule
   center=<-6.0501 3.578 1.0799>
   rotation=<rot 3.9219 0.7522 91.3132>
  }]
 foot_r [{
   shapeType=Capsule
   length=9.3648
   radius=5.603
   shapeName=foot_r_capsule
   center=<6.0357 -3.5869 -1.0809>
   rotation=<rot 4.1701 0.7289 -88.8871>
  }]
 hand_l [{
   shapeType=Capsule
   length=13.2473
   radius=3.551
   shapeName=hand_l_capsule
   center=<3.7727 -2.1515 -0.9836>
   rotation=<rot 83.9243 -40.1215 -43.9756>
  }]
 hand_r [{
   shapeType=Capsule
   length=13.3676
   radius=3.5328
   shapeName=hand_r_capsule
   center=<-3.7011 2.2846 1.0188>
   rotation=<rot 83.8857 -44.2443 130.8924>
  }]
 head [{
   shapeType=Capsule
   length=-0.1455
   radius=14.4032
   shapeName=head_capsule
   center=<-1.222 2.5308 -0.041>
   rotation=<rot -0.5145 14.924 -88.8585>
  }]
 lowerarm_l [{
   shapeType=Capsule
   length=17.7644
   radius=6.7712
   shapeName=lowerarm_l_capsule
   center=<8.6907 1.5074 -1.0415>
   rotation=<rot 83.9602 37.8123 -137.9838>
  }]
 lowerarm_r [{
   shapeType=Capsule
   length=17.7517
   radius=6.7956
   shapeName=lowerarm_r_capsule
   center=<-8.5185 -1.5118 1.0716>
   rotation=<rot 83.9085 43.1239 47.3384>
  }]
 neck_02 [{shapeType=Capsule, length=1, radius=1}]
 pelvis [{
   shapeType=Capsule
   length=16.6026
   radius=34.35
   shapeName=pelvis_capsule
   center=<4.7071 -3.4023 0.0456>
   rotation=<rot -88.9437 91.7789 -99.1541>
  }]
 spine_03 [{
   shapeType=Capsule
   length=17.6069
   radius=17.2759
   shapeName=spine_03_capsule
   center=<14.9822 -3.5844 -0.2106>
   rotation=<rot 0.3554 22.047 -0.9627>
  }]
 spine_05 [{
   shapeType=Capsule
   length=3.1063
   radius=22.4514
   shapeName=spine_05_capsule
   center=<-8.6883 0.5608 -0.5336>
   rotation=<rot -2.9097 2.0399 -0.3687>
  }]
 thigh_l [{
   shapeType=Capsule
   length=-6.2394
   radius=13.5302
   shapeName=thigh_l_capsule
   center=<-19.042 -1.4357 0.5263>
   rotation=<rot -70.231 -5.1202 5.8901>
  }]
 thigh_r [{
   shapeType=Capsule
   length=-6.0183
   radius=13.5083
   shapeName=thigh_r_capsule
   center=<19.0769 1.4465 -0.5305>
   rotation=<rot -70.0773 -5.2725 -173.9834>
  }]
 upperarm_l [{
   shapeType=Capsule
   length=19.3172
   radius=8.6965
   shapeName=upperarm_l_capsule
   center=<8.0631 -1.3129 -1.7568>
   rotation=<rot -8.2597 95.3671 -84.8991>
  }]
 upperarm_r [{
   shapeType=Capsule
   length=20.0155
   radius=8.6689
   shapeName=upperarm_r_capsule
   center=<-8.139 1.327 1.7564>
   rotation=<rot -7.8193 95.6805 95.1286>
  }]
}
constraints [
 {bone1Name=thigh_r, bone2Name=pelvis, swing1LimitDegrees=45, swing2LimitDegrees=45, twistLimitDegrees=45}
 {bone1Name=calf_r, bone2Name=thigh_r, swing1LimitDegrees=45, swing2LimitDegrees=45, twistLimitDegrees=45}
 {bone1Name=thigh_l, bone2Name=pelvis, swing1LimitDegrees=45, swing2LimitDegrees=45, twistLimitDegrees=45}
 {bone1Name=calf_l, bone2Name=thigh_l, swing1LimitDegrees=45, swing2LimitDegrees=45, twistLimitDegrees=45}
 {bone1Name=spine_03, bone2Name=pelvis, swing1LimitDegrees=45, swing2LimitDegrees=45, twistLimitDegrees=45}
 {bone1Name=spine_05, bone2Name=spine_03, swing1LimitDegrees=45, swing2LimitDegrees=45, twistLimitDegrees=45}
 {bone1Name=neck_02, bone2Name=spine_05, swing1LimitDegrees=45, swing2LimitDegrees=45, twistLimitDegrees=45}
 {bone1Name=upperarm_l, bone2Name=spine_05, swing1LimitDegrees=45, swing2LimitDegrees=45, twistLimitDegrees=45}
 {bone1Name=hand_l, bone2Name=lowerarm_l, swing1LimitDegrees=45, swing2LimitDegrees=45, twistLimitDegrees=45}
 {bone1Name=upperarm_r, bone2Name=spine_05, swing1LimitDegrees=45, swing2LimitDegrees=45, twistLimitDegrees=45}
 {bone1Name=hand_r, bone2Name=lowerarm_r, swing1LimitDegrees=45, swing2LimitDegrees=45, twistLimitDegrees=45}
 {
  bone1Name=head
  bone2Name=neck_02
  swing1LimitDegrees=45
  swing1Motion=Free
  swing2LimitDegrees=45
  swing2Motion=Free
  twistLimitDegrees=45
  twistMotion=Free
 }
 {
  bone1Name=foot_l
  bone2Name=calf_l
  swing1LimitDegrees=45
  swing1Motion=Free
  swing2LimitDegrees=45
  swing2Motion=Free
  twistLimitDegrees=45
  twistMotion=Free
 }
 {
  bone1Name=lowerarm_l
  bone2Name=upperarm_l
  swing1LimitDegrees=45
  swing1Motion=Free
  swing2LimitDegrees=45
  swing2Motion=Free
  twistLimitDegrees=45
  twistMotion=Free
 }
 {
  bone1Name=lowerarm_r
  bone2Name=upperarm_r
  swing1LimitDegrees=45
  swing1Motion=Free
  swing2LimitDegrees=45
  swing2Motion=Free
  twistLimitDegrees=45
  twistMotion=Free
 }
]
